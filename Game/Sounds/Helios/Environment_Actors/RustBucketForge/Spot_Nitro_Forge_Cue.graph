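# Ambient looping sound cue for the Nitro Forge environment actor, reusing RustBucket propane torch fire stand loops pitched down.
# A Random node picks one of three equally weighted looping waves, and a Modulator varies its pitch slightly.
# Inline attenuation overrides add logarithmic falloff, occlusion and listener focus, while overridden concurrency limits simultaneous instances.
# Audio routes to the Athena ambient props class, exterior ambience bus and default reverb submix.
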SoundCue /Game/Sounds/Helios/Environment_Actors/RustBucketForge/Spot_Nitro_Forge_Cue
root=Modulator_1
nodes [
 {id=Modulator_1, children [Random_1], properties {pitchMax=1.1, pitchMin=0.9}}
 {id=Random_1, children [WavePlayer_4,WavePlayer_5,WavePlayer_6], properties {weights [1,1,1]}}
 {
  id=WavePlayer_4
  properties {
   bLooping=true
   soundWaveAssetPtr=/Game/Sounds/Helios/Environment_Actors/RustBucketTorchFireStand/Spot_RustBucket_Propane_Torch_Fire_Stand_Loop_03_Cue
  }
 }
 {
  id=WavePlayer_5
  properties {
   bLooping=true
   soundWaveAssetPtr=/Game/Sounds/Helios/Environment_Actors/RustBucketTorchFireStand/Spot_RustBucket_Propane_Torch_Fire_Stand_Loop_02_Cue
  }
 }
 {
  id=WavePlayer_6
  properties {
   bLooping=true
   soundWaveAssetPtr=/Game/Sounds/Helios/Environment_Actors/RustBucketTorchFireStand/Spot_RustBucket_Propane_Torch_Fire_Stand_Loop_01_Cue
  }
 }
]
properties {
 attenuationSettings=/Game/Sounds/Helios/Environment_Actors/RustBucketTorchFireStand/Spot_RustBucket_Torch_Fire_Stand_Att
 bBypassVolumeScaleForPriority=true
 bOverrideAttenuation=true
 bOverrideConcurrency=true
 maxDistance=2750
 pitchMultiplier=0.7
 priority=0
 soundClassObject=&Fort_Ambient_Props_Athena
 volumeMultiplier=0.752
 attenuationOverrides {
  bEnableFocusInterpolation=true
  bEnableListenerFocus=true
  bEnableOcclusion=true
  distanceAlgorithm=Logarithmic
  falloffDistance=2500
  focusAttackInterpSpeed=0.9
  focusAzimuth=15
  focusReleaseInterpSpeed=0.4
  focusVolumeAttenuation=1.2
  nonFocusAzimuth=45
  nonFocusVolumeAttenuation=0.2
  stereoSpread=10
  attenuationShapeExtents {x=250}
 }
 busSends [{audioBus=null, minSendLevel=1, sendLevel=1, soundSourceBus=&AmbientAudio_Exterior_SourceBus}]
 concurrencyOverrides {maxCount=2, voiceStealReleaseTime=2}
 soundSubmixSends [{minSendLevel=0.6, soundSubmix=&Fortnite_Reverb_Default}]
}
